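# Translucent unlit Niagara sprite material that draws a procedural crescent moon for the LexaEarlGrey pickaxe FX.
# A UV-space sphere mask has a second one subtracted from it, whose center and radius follow the Moon Phase dynamic parameter.
# Depth fade softens the shape except at Low quality.
# The shape scales particle alpha and, through SmoothStep, blends particle color toward black before MF_FX_EmissiveAdjustments.
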Material /BRCosmetics/Weapons/FORT_Melee/Pickaxe_LexaEarlGrey/FX/M_Pickaxe_LexaEarlGrey_Moon
graph {
 properties {bUsedWithNiagaraSprites=true, blendMode=BLEND_Translucent, pixelDepthOffsetMode=PDOM_Legacy, shadingModel=MSM_Unlit, usedShadingModels=Unlit}
 nodes [
  n1 = ScalarParameter() {parameterName=Min}
  n2 = ScalarParameter() {defaultValue=1, parameterName=Max}
  n3 = TextureCoordinate()
  n4 = Constant() {r=0.5}
  n5 = Constant() {r=0.5}
  n6 = ScalarParameter() {defaultValue=0.9, parameterName=Outer Hardness, sliderMax=1}
  n7 = DynamicParameter() {defaultValue=<rgba 0.25 1 1 1>, paramNames [Moon Phase, Param2, Param3, Param4]}
  n8 = ScalarParameter() {defaultValue=0.9, parameterName=Inner HArdness, sliderMax=1}
  n9 = DepthFade(FadeDistance=ScalarParameter() {defaultValue=8, parameterName=Depth Fade})
  n10 = Multiply(
   A=Subtract(
    A=SphereMask(A=@n3, B=@n4, Radius=@n5, Hardness=@n6) {hardnessPercent=90}
    B=SphereMask(A=@n3, B=@n7.Moon Phase, Radius=@n7.Moon Phase, Hardness=@n8) {hardnessPercent=90}
   )
   B=QualitySwitch(Default=@n9, Low=Constant() {r=1}, High=@n9, Medium=@n9)
  )
  n11 = ParticleColor()
  MaterialFunctionCall(Emissive=LinearInterpolate(A=@n11.RGB, B=Constant3Vector(), Alpha=SmoothStep(Min=@n1, Max=@n2, Value=@n10) {constMax=0.5})) {materialFunction=&MF_FX_EmissiveAdjustments}
  Multiply(A=@n11.A, B=@n10)
 ]
}
