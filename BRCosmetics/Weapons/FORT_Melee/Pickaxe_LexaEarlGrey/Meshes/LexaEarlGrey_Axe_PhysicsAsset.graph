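# Physics asset for the LexaEarlGrey axe pickaxe, a two-handed melee cosmetic weapon mesh.
# It holds a single capsule body on the Melee_TwoHanded_Handle bone, offset along the handle and rotated to lie along it.
# There are no constraints, so the weapon collides and simulates as one rigid piece.
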PhysicsAsset /BRCosmetics/Weapons/FORT_Melee/Pickaxe_LexaEarlGrey/Meshes/LexaEarlGrey_Axe_PhysicsAsset
bodies {Melee_TwoHanded_Handle [{shapeType=Capsule, length=45.7321, radius=12.0374, center=<0.0006 3.1533 20.9505>, rotation=<rot -2.183 -90 0.0029>}]}
constraints []
